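; AI random-targeter controller component for the FirePetal Wandering Boss bot, deciding when it may switch targets and which targets it skips.
; Target switching is allowed unless the owning bot controller is at Threatened alert and its boss pawn carries the Ability.WanderingBoss.TargetLock gameplay tag.
; Downed (DBNO) pawns are ignored as targets, and holster application is always refused.
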(event EventBeginPlay)

(event EventTick
  (params (DeltaSeconds :type Float%20%28single-precision%29)))

(fn AllowedToSwitchTarget
  (bind _asfort_athena_aibot_controller (Utilities|Casting|CastToFortAthenaAIBotController (Components|GetOwner)))
  (bind _output_get (LocalVariables|Setloc_OwnerAIBotController _asfort_athena_aibot_controller))
  (bind _asbp_player_pawn_fire_petal_wandering_boss (Utilities|Casting|CastToBP_PlayerPawn_FirePetal_WanderingBoss (Pawn|GetControlledPawn _output_get)))
  (LocalVariables|Setloc_OwnerPawn _asbp_player_pawn_fire_petal_wandering_boss)
  (switch Utilities|FlowControl|Switch|SwitchonEAlertLevel (Variables|/Script/FortniteAI.FortAthenaAIBotController|GetCurrentAlertLevel (LocalVariables|Getloc_OwnerAIBotController))
    (:Unaware
      _)
    (:Suspicious
      _)
    (:LostTarget
      _)
    (:Threatened
      (Utilities|FlowControl|Sequence
        (:then_0
          (if (GameplayTags|HasMatchingGameplayTag (LocalVariables|Getloc_OwnerPawn) "(TagName=\"Ability.WanderingBoss.TargetLock\")")
            (return false)))
        (:then_1
          (return true))))
    (:Count
      _))
  (return true))

(fn CanApplyHolster
  (params (HolsterId :type Name%20%28by%20ref%29 :default "None"))
  (return false))

(fn TargetIsIgnored
  (params (Target :type Actor%20Object%20Reference))
  (bind _asfort_pawn (Utilities|Casting|CastToFortPawn Target))
  (return (Health|IsDBNO _asfort_pawn)))
